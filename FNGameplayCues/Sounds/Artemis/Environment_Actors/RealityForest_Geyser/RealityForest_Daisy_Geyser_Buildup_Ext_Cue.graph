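# Sound cue for the exterior buildup sound of Reality Forest geysers, classed as an outdoor ambient one-shot.
# A Random node picks one of four equally weighted buildup waves and feeds it through a root Modulator.
# It uses the smaller geyser-launch attenuation asset, and overridden concurrency caps simultaneous instances at a low count.
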SoundCue /FNGameplayCues/Sounds/Artemis/Environment_Actors/RealityForest_Geyser/RealityForest_Daisy_Geyser_Buildup_Ext_Cue
root=Modulator_0
nodes [
 {id=Modulator_0, children [Random_1]}
 {
  id=Random_1
  children [WavePlayer_4,WavePlayer_5,WavePlayer_6,WavePlayer_7]
  properties {bShouldExcludeFromBranchCulling=true, weights [1,1,1,1]}
 }
 {
  id=WavePlayer_4
  properties {soundWaveAssetPtr=/Game/Sounds/Artemis/Environment_Actors/RealityForest_Geyser/AMB_RealityForest_Geyser_Buildup_01}
 }
 {
  id=WavePlayer_5
  properties {soundWaveAssetPtr=/Game/Sounds/Artemis/Environment_Actors/RealityForest_Geyser/AMB_RealityForest_Geyser_Buildup_02}
 }
 {
  id=WavePlayer_6
  properties {soundWaveAssetPtr=/Game/Sounds/Artemis/Environment_Actors/RealityForest_Geyser/AMB_RealityForest_Geyser_Buildup_03}
 }
 {
  id=WavePlayer_7
  properties {soundWaveAssetPtr=/Game/Sounds/Artemis/Environment_Actors/RealityForest_Geyser/AMB_RealityForest_Geyser_Buildup_04}
 }
]
properties {
 attenuationSettings=&/RealityForest_Daisy_Geyser_Launch_Smaller_Att
 bOverrideConcurrency=true
 maxDistance=9750
 soundClassObject=&Fort_Ambient_OneShots_Outdoors
 volumeMultiplier=0.8
 attenuationOverrides {bEnableListenerFocus=true, falloffDistance=1600, nonFocusVolumeAttenuation=0.5}
 concurrencyOverrides {maxCount=2}
}
