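# Sound cue for the Reality Forest daisy geyser launch, played as an outdoor ambient one-shot with attenuation.
# A mixer layers one of four random geyser blast waves, at slightly reduced volume, with one of four random first-person whoosh waves delayed by a fixed fraction of a second.
# Each layer passes through its own modulator for variation.
# A concurrency override caps the cue at two simultaneous instances.
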SoundCue /FNGameplayCues/Sounds/Artemis/Environment_Actors/RealityForest_Geyser/RealityForest_Daisy_Geyser_Launch_Cue
root=Mixer_0
nodes [
 {id=Delay_0, children [Modulator_1], properties {delayMax=0.3, delayMin=0.3}}
 {id=Mixer_0, children [Modulator_0,Delay_0], properties {inputVolume [0.8,1]}}
 {id=Modulator_0, children [Random_0]}
 {id=Modulator_1, children [Random_1]}
 {id=Random_0, children [WavePlayer_0,WavePlayer_1,WavePlayer_2,WavePlayer_3], properties {weights [1,1,1,1]}}
 {id=Random_1, children [WavePlayer_4,WavePlayer_5,WavePlayer_6,WavePlayer_7], properties {weights [1,1,1,1]}}
 {
  id=WavePlayer_0
  properties {soundWaveAssetPtr=&/AMB_RealityForest_Geyser_Blast_04}
 }
 {
  id=WavePlayer_1
  properties {soundWaveAssetPtr=&/AMB_RealityForest_Geyser_Blast_01}
 }
 {
  id=WavePlayer_2
  properties {soundWaveAssetPtr=&/AMB_RealityForest_Geyser_Blast_02}
 }
 {
  id=WavePlayer_3
  properties {soundWaveAssetPtr=&/AMB_RealityForest_Geyser_Blast_03}
 }
 {
  id=WavePlayer_4
  properties {soundWaveAssetPtr=&/AMB_RealityForest_Geyser_Blast_1p_Whoosh_01}
 }
 {
  id=WavePlayer_5
  properties {soundWaveAssetPtr=&/AMB_RealityForest_Geyser_Blast_1p_Whoosh_02}
 }
 {
  id=WavePlayer_6
  properties {soundWaveAssetPtr=&/AMB_RealityForest_Geyser_Blast_1p_Whoosh_03}
 }
 {
  id=WavePlayer_7
  properties {soundWaveAssetPtr=&/AMB_RealityForest_Geyser_Blast_1p_Whoosh_04}
 }
]
properties {
 attenuationSettings=/Game/Sounds/Artemis/Environment_Actors/RealityForest_Geyser/RealityForest_Daisy_Geyser_Launch_Att
 bOverrideConcurrency=true
 maxDistance=5500
 soundClassObject=&Fort_Ambient_OneShots_Outdoors
 volumeMultiplier=0.7
 attenuationOverrides {bEnableListenerFocus=true, falloffDistance=2000, attenuationShapeExtents {x=500}}
 concurrencyOverrides {maxCount=2}
}
